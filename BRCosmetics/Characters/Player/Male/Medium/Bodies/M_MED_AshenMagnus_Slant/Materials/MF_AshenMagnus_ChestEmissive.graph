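# Material function that adds an animated, flame-like emissive effect to the Ashen Magnus character's chest, toggled by the UseChestEffect static switch.
# Polar UVs built from pre-skinned local position feed panning, noise-distorted textures to form a gradient, simplified on forward and mobile paths.
# Outer and core curve-atlas colors blend by that gradient and mask channels, replacing emissive within a box-limited chest mask.
# The highlight and time-of-day tint branches end unconnected.
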MaterialFunction /BRCosmetics/Characters/Player/Male/Medium/Bodies/M_MED_AshenMagnus_Slant/Materials/MF_AshenMagnus_ChestEmissive
graph {nodes [
  n1 = NamedRerouteUsage()
  n2 = Constant2Vector() {g=0.1}
  n3 = Multiply(
   A=NamedRerouteUsage()
   B=ComponentMask(None=VectorParameter() {group=AshenMagnus_ChestEffect, parameterName=Noise_base_UV}.RGB) {g=true, r=true}
  )
  n4 = Constant2Vector() {g=0.35}
  n5 = ComponentMask(None=TextureSample(UVs=Panner(Coordinate=@n3, Speed=@n4) {speedY=0.35}) {samplerSource=SSM_Wrap_WorldGroupSettings, texture=&T_NoisePack_01}.RGB) {g=true, r=true}
  n6 = TextureSampleParameter2D(UVs=Add(A=Panner(Coordinate=@n1, Speed=@n2) {speedY=0.1}, B=Multiply(A=@n5) {constB=0.2})) {
   group=AshenMagnus_ChestEffect
   parameterName=HighLight_Texture
   samplerSource=SSM_Wrap_WorldGroupSettings
   texture=/Game/Global/Textures/Default/Noise/T_SoftCells_Negative
  }
  n7 = LinearInterpolate(Alpha=@n6.R) {constA=-0.2}
  n8 = Constant() {r=1}
  n9 = Constant() {r=3}
  n10 = OneMinus(None=ComponentMask(None=NamedRerouteUsage()) {g=true})
  n11 = Multiply(A=@n7, B=Power(Base=@n7, Exp=LinearInterpolate(A=@n8, B=@n9, Alpha=@n10) {constA=0.6, constB=4}))
  n12 = VectorParameter() {group=AshenMagnus_ChestEffect, parameterName=HighLight_color, defaultValue=<rgba 0.0123 0.8879 0.6445 1>}
  n13 = ComponentMask(None=NamedRerouteUsage()) {g=true}
  n14 = Power(Base=Multiply(A=Multiply(A=@n11, B=@n12.RGB), B=Power(Base=@n13) {constExponent=5}))
  n15 = Constant()
  n16 = NamedRerouteUsage()
  n17 = Multiply(
   A=NamedRerouteUsage()
   B=ComponentMask(None=VectorParameter() {group=AshenMagnus_ChestEffect, parameterName=Default_texture_UV, defaultValue=<rgba 1 0.75 6 0.5>}.RGB) {g=true, r=true}
  )
  n18 = Constant2Vector() {g=0.2, r=0.1}
  n19 = Constant() {r=0.5}
  n20 = Multiply(
   A=NamedRerouteUsage()
   B=ComponentMask(None=VectorParameter() {group=AshenMagnus_ChestEffect, parameterName=Disturb_texture_UV, defaultValue=<rgba 2 0.75 6 0.5>}.RGB) {g=true, r=true}
  )
  n21 = Constant2Vector() {g=0.2, r=0.03}
  n22 = Multiply(
   A=NamedRerouteUsage()
   B=ComponentMask(None=VectorParameter() {group=AshenMagnus_ChestEffect, parameterName=Noise_base_UV}.RGB) {g=true, r=true}
  )
  n23 = TextureSample(UVs=Panner(Coordinate=@n22, Speed=@n21) {speedX=0.03, speedY=0.2}) {
   samplerSource=SSM_Wrap_WorldGroupSettings
   samplerType=SAMPLERTYPE_LinearColor
   texture=&T_NoisePack_LinearColor
  }
  n24 = ComponentMask(None=@n23.RGB) {g=true, r=true}
  n25 = TextureSampleParameter2D(UVs=Add(A=Panner(Coordinate=@n20, Speed=@n21) {speedX=0.03, speedY=0.2}, B=Multiply(A=@n24) {constB=0.05})) {
   group=AshenMagnus_ChestEffect
   parameterName=Disturb_texture
   samplerSource=SSM_Wrap_WorldGroupSettings
   samplerType=SAMPLERTYPE_Normal
   texture=/Game/Global/Textures/Default/Noise/T_FlowWarp_Noise
  }
  n26 = ComponentMask(None=NamedRerouteDeclaration(None=@n25.RGB) {name=DisturbNormal, nodeColor=<rgba 0.2891 1 0 1>}) {g=true, r=true}
  n27 = TextureSample(UVs=Add(A=Panner(Coordinate=@n17, Speed=@n18) {speedX=0.1, speedY=0.2}, B=Multiply(A=@n19, B=@n26))) {samplerSource=SSM_Wrap_WorldGroupSettings, samplerType=SAMPLERTYPE_Masks, texture=/Game/Global/Textures/Default/Noise/T_WrapEdgeWear_Noise}
  n28 = Add(A=Subtract(A=@n27.R) {constB=0.5}, B=ComponentMask(None=NamedRerouteUsage()) {g=true})
  n29 = Constant() {r=1}
  n30 = OneMinus(None=Abs(None=@n25.R))
  n31 = Constant()
  n32 = Multiply(
   A=NamedRerouteUsage()
   B=ComponentMask(None=VectorParameter() {group=AshenMagnus_ChestEffect, parameterName=Flame_Texture_UV, defaultValue=<rgba 1.5 0.5 0 0>}.RGB) {g=true, r=true}
  )
  n33 = Constant2Vector() {g=0.55}
  n34 = Constant2Vector() {g=0.35}
  n35 = ComponentMask(None=TextureSample(UVs=Panner(Coordinate=@n22, Speed=@n34) {speedY=0.35}) {samplerSource=SSM_Wrap_WorldGroupSettings, texture=&T_NoisePack_01}.RGB) {g=true, r=true}
  n36 = TextureSampleParameter2D(UVs=Add(A=Panner(Coordinate=@n32, Speed=@n33) {speedY=0.55}, B=Multiply(A=@n35) {constB=0.1})) {
   group=AshenMagnus_ChestEffect
   parameterName=Flame_Texture
   samplerSource=SSM_Wrap_WorldGroupSettings
   texture=/Game/Weapons/WeaponSkins/Wraps/Textures/PlasmaClouds/T_Wrap_PlasmaClouds__M
  }
  n37 = NamedRerouteUsage()
  n38 = LinearInterpolate(
   A=ScalarParameter() {group=AshenMagnus_ChestEffect, parameterName=Streaks_LerpA}
   B=ScalarParameter() {defaultValue=3, group=AshenMagnus_ChestEffect, parameterName=Streaks_LerpB}
   Alpha=OneMinus(None=ComponentMask(None=@n37) {g=true})
  ) {constA=0.6, constB=4}
  n39 = ComponentMask(None=NamedRerouteUsage()) {g=true}
  n40 = ScalarParameter() {group=AshenMagnus_ChestEffect, parameterName=Central Contast}
  n41 = Add(
   A=ShadingPathSwitch(Default=Multiply(A=@n30, B=@n30), Forward=@n31, Mobile=@n31)
   B=Multiply(A=Power(Base=@n36.G, Exp=@n38), B=Power(Base=@n39, Exp=@n40) {constExponent=10})
  )
  n42 = NamedRerouteDeclaration(None=Multiply(A=ShadingPathSwitch(Default=@n28, Forward=@n29, Mobile=@n29), B=@n41) {constA=1}) {name=EffectGradient, nodeColor=<rgba 0.0312 0 1 1>}
  n43 = MaterialFunctionCall(Float3=NamedRerouteUsage()) {materialFunction=&BreakOutFloat3Components}
  n44 = ScalarParameter() {defaultValue=1, group=AshenMagnus_ChestEffect, parameterName=RedEmissiveMaskInfluence}
  n45 = Add(
   A=Subtract(A=@n42, B=Multiply(A=@n43.R, B=@n44))
   B=Multiply(A=@n43.B, B=ScalarParameter() {group=AshenMagnus_ChestEffect, parameterName=BlueEmissiveMaskInfluence})
  )
  n46 = CurveAtlasRowParameter(CurveTime=Saturate(None=Multiply(A=@n16, B=@n45))) {
   atlas=/BRCosmetics/Characters/Player/Male/Medium/Bodies/M_MED_AshenMagnus_Slant/Skins/Rye/Textures/CA_AshenMagnus_ChestPiece
   curve=/BRCosmetics/Characters/Player/Male/Medium/Bodies/M_MED_AshenMagnus_Slant/Skins/Rye/Textures/CC_AshenMagnus_ChestPiece_Color
   group=AshenMagnus_ChestEffect
  }
  n47 = Multiply(A=LinearInterpolate(A=@n46, Alpha=ScalarParameter() {group=AshenMagnus_ChestEffect, parameterName=Color_lerp}), B=@n41)
  Add(A=ShadingPathSwitch(Default=Multiply(A=@n14) {constB=2}, Forward=@n15, Mobile=@n15), B=@n47)
  n48 = FunctionInput() {inputName=Material, inputType=FunctionInput_MaterialAttributes}
  n49 = GetMaterialAttributes(None=@n48) {attributeGetTypes [#1, #2, #3]}
  n50 = SetMaterialAttributes!(MaterialAttributes=@n49.MaterialAttributes, EmissiveColor=NamedRerouteUsage()) {attributeSetTypes [#1]}
  n51 = MaterialFunctionCall(Float3=NamedRerouteUsage()) {materialFunction=&BreakOutFloat3Components}
  n52 = Max(A=@n51.R, B=@n51.R)
  n53 = TextureCoordinate()
  n54 = Constant2Vector() {g=0.215, r=0.365}
  n55 = Constant2Vector() {g=0.1, r=0.1}
  n56 = Constant()
  n57 = Multiply(A=Multiply(
    A=Max(A=@n52, B=@n51.B)
    B=MaterialFunctionCall!(A=@n53, B=@n54, Bounds=@n55, Edge Falloff=@n56) {materialFunction=/Engine/Functions/Engine_MaterialFunctions01/Gradient/BoxMask-2D}.Result
   )) {constB=2}
  n58 = FunctionInput(Preview=NamedRerouteDeclaration(None=Saturate(None=@n57)) {name=ChestMask, nodeColor=<rgba 1 0 0.5391 1>}) {bUsePreviewValueAsDefault=true, inputName=Mask, inputType=FunctionInput_Scalar, sortPriority=1}
  FunctionOutput(None=StaticSwitchParameter(True=BlendMaterialAttributes(A=@n49.MaterialAttributes, B=@n50, Alpha=@n58), False=@n48) {group=AshenMagnus_ChestEffect, parameterName=UseChestEffect})
  Multiply(
   A=MaterialFunctionCall(
    Color_Morning=VectorParameter() {group=AshenMagnus_ChestEffect, parameterName=Morning_Color, defaultValue=<rgba 1 1 1 1>}.RGB
    Color_Day=VectorParameter() {group=AshenMagnus_ChestEffect, parameterName=Day_Color, defaultValue=<rgba 1 1 1 1>}.RGB
    Color_Evening=VectorParameter() {group=AshenMagnus_ChestEffect, parameterName=Evening_Color, defaultValue=<rgba 1 1 1 1>}.RGB
    Color_Night=VectorParameter() {group=AshenMagnus_ChestEffect, parameterName=Night_Color, defaultValue=<rgba 1 1 1 1>}.RGB
   ) {materialFunction=&Color_Time_of_Day}.Result
   B=@n47
  )
  n59 = CurveAtlasRowParameter(CurveTime=Saturate(None=@n42)) {
   atlas=&CA_AshenMagnus_ChestPiece
   curve=/BRCosmetics/Characters/Player/Male/Medium/Bodies/M_MED_AshenMagnus_Slant/Skins/Rye/Textures/CC_AshenMagnus_ChestPiece_OuterColors
   defaultValue=2
   group=AshenMagnus_ChestEffect
   parameterName=OuterBrightColors
  }
  n60 = CurveAtlasRowParameter(CurveTime=Saturate(None=Multiply(A=@n42))) {
   atlas=&CA_AshenMagnus_ChestPiece
   curve=/BRCosmetics/Characters/Player/Male/Medium/Bodies/M_MED_AshenMagnus_Slant/Skins/Rye/Textures/CC_AshenMagnus_ChestPiece_CoreColors
   defaultValue=1
   group=AshenMagnus_ChestEffect
   parameterName=CoreBrightColors
  }
  n61 = NamedRerouteUsage()
  n62 = MaterialFunctionCall(Float3=NamedRerouteUsage()) {materialFunction=&BreakOutFloat3Components}
  n63 = Multiply(A=@n62.B, B=ScalarParameter() {group=AshenMagnus_ChestEffect, parameterName=BlueEmissiveMaskInfluence})
  n64 = ScalarParameter() {group=AshenMagnus_ChestEffect, parameterName=RedEmissiveMaskInfluence}
  n65 = Power(Base=Subtract(A=Add(A=@n61, B=@n63), B=Multiply(A=@n62.R, B=@n64))) {constExponent=3}
  NamedRerouteDeclaration(None=LinearInterpolate(A=@n59, B=@n60, Alpha=@n65)) {name=OUT_Colors, nodeColor=<rgba 0 1 0.0625 1>}
  NamedRerouteDeclaration(None=@n49.EmissiveColor) {name=IN_Emissive, nodeColor=<rgba 0 1 0.2734 1>}
  NamedRerouteDeclaration(None=@n27.R) {name=GradientFullRange, nodeColor=<rgba 0 0.6484 1 1>}
  n66 = Subtract(
   A=VectorParameter() {group=AshenMagnus_ChestEffect, parameterName=RadialUV_Offset}.RGB
   B=VertexInterpolator(VS=LocalPosition() {localOrigin=InstancePreSkinning}.XYZ).PS
  )
  n67 = MaterialFunctionCall(Float2=MaterialFunctionCall(
    UVs=ComponentMask(None=@n66) {b=true, r=true}
    Texture Scale=VectorParameter() {group=AshenMagnus_ChestEffect, parameterName=RadialUV_Scale, defaultValue=<rgba 1 1 1 0>}.RGB
   ) {materialFunction=&ScaleUVsByCenter}.UVs) {materialFunction=&BreakOutFloat2Components}
  n68 = ComponentMask(None=@n66) {g=true}
  n69 = StaticComponentMaskParameter(None=Divide(
    A=MaterialFunctionCall(X=@n67.R, Y=@n68, Z=@n67.R) {materialFunction=&MakeFloat3}.Result
    B=PreSkinnedLocalBounds().Extents
   )) {defaultB=true, defaultR=true, group=AshenMagnus_ChestEffect, parameterName=Area01Channel}
  n70 = Divide(
   A=Arctangent2Fast(Y=ComponentMask(None=@n69) {g=true}, X=ComponentMask(None=@n69) {r=true})
   B=MaterialFunctionCall() {materialFunction=&Pi}.Result
  )
  n71 = Distance(A=@n69, B=Constant())
  n72 = StaticSwitchParameter(True=ScalarParameter() {defaultValue=1, group=AshenMagnus_ChestEffect, parameterName=StaticRadius}, False=ObjectRadius()) {defaultValue=true, group=AshenMagnus_ChestEffect, parameterName=UseStaticRadius}
  n73 = AppendVector(A=Multiply(A=Add(A=@n70) {constB=-0.5}) {constB=0.5}, B=OneMinus(None=Divide(A=@n71, B=@n72)))
  n74 = ComponentMask(None=VectorParameter() {group=AshenMagnus_ChestEffect, parameterName=RadialUV_Mult, defaultValue=<rgba 1 1 0 0>}.RGB) {g=true, r=true}
  NamedRerouteDeclaration(None=Frac(None=Multiply(A=@n73, B=@n74))) {name=RadialUV, nodeColor=<rgba 1 0.2812 0 1>}
  NamedRerouteDeclaration(None=@n49.Normal) {name=IN_Normal, nodeColor=<rgba 1 0 0.6094 1>}
  NamedRerouteDeclaration(None=@n49.BaseColor) {name=IN_BaseColor, nodeColor=<rgba 0.0078 0 1 1>}
  Saturate(None=@n65)
  TextureCoordinate()
 ]}
