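# Animation blueprint for the Ashen Magnus Slant player body cosmetic, copying the parent mesh pose with the jaw taken from the Face mesh.
# A Control Rig receives sliding, crouch, emote/float and parachute states.
# Calf-relative foot yaw then weights additive offsets on the hydraulics bones; their constraint nodes are unconnected.
# Finally, DefaultSlot blends onto the pistol bones, weighted by the Frontend_Idle_Bypass curve read in the front end.
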AnimBlueprint /BRCosmetics/Characters/Player/Male/Medium/Bodies/M_MED_AshenMagnus_Slant/Meshes/M_MED_AshenMagnus_Slant_AnimBP
target_skeleton=&/M_MED_AshenMagnus_Slant_Skeleton
graphs [
 AnimGraph {nodes [
   n1 = "Constraint - hydraulics_l (P)"(Alpha=1, ConstraintWeights_0=1) {
    alphaBoolBlend {blendOption=Linear}
    boneToModify {boneName=hydraulics_l}
    constraintSetup [{targetBone {boneName=calf_l}}]
    constraintWeights [1]
   }
   "Constraint - hydraulics_r (P)"(ComponentPose=@n1, Alpha=1, ConstraintWeights_0=1) {
    alphaBoolBlend {blendOption=Linear}
    boneToModify {boneName=hydraulics_r}
    constraintSetup [{targetBone {boneName=calf_r}}]
    constraintWeights [1]
   }
   n2 = Get Body_Mesh()
   n3 = Control Rig!(
    Source=Layered blend per bone(
     BasePose=Copy Pose From Mesh() {bUseAttachedParent=true}
     BlendPoses_0=Copy Pose From Mesh(SourceMeshComponent=@n2) {bUseAttachedParent=true}
    ) {layerSetup [{branchFilters [{boneName=C_jaw}]}]}
    Alpha
    is sliding:bool=Get bIsSliding()
    is_crouch:double=Get Is Crouch() -> (Is Crouch:double)
    is_Fly_Emote:double=Get Is Fly Emote() -> (Is Fly Emote:double)
    is_Parachute:double=Get Is Parachute() -> (Is Parachute:double)
   ) -> (Pose) {
    alphaBoolBlend {blendOption=Linear}
    controlRigAssetReference {blueprintRigClass=&/M_MED_AshenMagnus_Slant_CtrlRig.M_MED_AshenMagnus_Slant_CtrlRig_C}
   }
   n4 = "Transform (Modify) Bone - Bone: hydraulics_l"(
    ComponentPose=Local To Component(LocalPose=@n3)
    Alpha=Get foot rot L() -> (foot rot L:double)
    Translation="0.000,-2.480,-3.000"
    Rotation="0,0,0"
    Scale="1,1,1"
   ) {
    translationMode=BMM_Additive
    translationSpace=BCS_BoneSpace
    alphaBoolBlend {blendOption=Linear}
    boneToModify {boneName=hydraulics_l}
    translation {y=-2.48, z=-3}
   }
   n5 = "Transform (Modify) Bone - Bone: hydraulics_r"(ComponentPose=@n4, Alpha=Get foot rot R() -> (foot rot R:double), Translation="0.061,0.000,-2.279", Rotation="0,0,0", Scale="1,1,1") {
    translationMode=BMM_Additive
    translationSpace=BCS_BoneSpace
    alphaBoolBlend {blendOption=Linear}
    boneToModify {boneName=hydraulics_r}
    translation {x=0.061, z=-2.279}
   }
   Output Pose(Result=Layered blend per bone(
     BasePose=Component To Local(ComponentPose=@n5)
     BlendPoses_0=Slot 'DefaultSlot'()
     BlendWeights_0=Get Custom Idle is Playing Float() -> (Custom Idle is Playing Float:double)
    ) {layerSetup [{branchFilters [{boneName=Pistol_l}, {boneName=Pistol_r}]}]}) {layerGroup=None}
  ]}
 EventGraph {nodes [
   Event OnCharacterCustomizationLoaded()
   > n1 = Cast To PlayerPawn_Athena(Object=TryGetPawnOwner())
   > Set Body_Mesh(Body_Mesh=GetSkeletalMeshForPartType(self=@n1, PartType=Face))
   Event BlueprintUpdateAnimation()
   > n2 = Sequence!(execute) -> (then_0, then_1, then_2:Exec, then_3:Exec)
   > n3 = Branch(Condition=Get bIsCrouching())
   > Set Is Crouch(Is Crouch:double=1) -> (Output_Get:double)
   Set Is Crouch(execute=@n3.else, Is Crouch=0)
   n4 = GetOwnerSkeletalMeshComponent()
   Set foot rot R(
    execute=@n2.then_1
    foot rot R:double=MapRangeClamped(
     Value=GetDeltaTransformFromRefPose!(self=@n4, BoneName=foot_r, BaseName=calf_r) -> (
      ReturnValue_Location:Vector
      ReturnValue_Rotation_Roll:float
      ReturnValue_Rotation_Pitch:float
      ReturnValue_Rotation_Yaw:float
      ReturnValue_Scale:Vector
     ).ReturnValue_Rotation_Yaw
     InRangeA=40
     InRangeB=-40
     OutRangeA=-1
    )
   ) -> (Output_Get:double)
   n5 = GetOwnerSkeletalMeshComponent()
   > Set foot rot L(foot rot L:double=MapRangeClamped(
     Value=GetDeltaTransformFromRefPose(self=@n5, BoneName=foot_l, BaseName=calf_l).ReturnValue_Rotation_Yaw
     InRangeA=40
     InRangeB=-40
     OutRangeA=-1
    )) -> (Output_Get:double)
   n6 = Branch(execute=@n2.then_2, Condition=OR Boolean(A=Get bIsPlayingEmote(), B=Get bIsSkydiveFloating()))
   > Set Is Fly Emote(Is Fly Emote:double=1) -> (Output_Get:double)
   Set Is Fly Emote(execute=@n6.else, Is Fly Emote=0)
   > n7 = Branch(Condition=Get bIsParachuteOpened())
   > Set Is Parachute(Is Parachute:double=1) -> (Output_Get:double)
   Set Is Parachute(execute=@n7.else, Is Parachute=0)
   Branch(execute=@n2.then_3, Condition=Get bIsInFrontEnd())
   > Set Custom Idle is Playing Float(Custom Idle is Playing Float:double=GetCurveValue(self=GetOwnerFrontEndAnimInstance(), CurveName=Frontend_Idle_Bypass)) -> (Output_Get:double)
   Get bIsSkyDiving()
  ]}
]
variables [
 Body_Mesh {type=SkeletalMeshComponent}
 Custom Idle is Playing Float {type=number}
 foot rot L {type=number}
 foot rot R {type=number}
 Is Crouch {type=number}
 Is Fly Emote {type=number}
 Is Parachute {type=number}
]
