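# Physics asset for the face accessory part of the male AshenMagnus Slant cosmetic, defining bodies and joint limits for dynamic hair bones.
# Each bone has one small sphere: a Hair_base root, three front strands, and left and right top sets of three.
# All nine strands constrain directly to Hair_base rather than chaining.
# Twist is nearly locked, swing is narrow, and top strands get a wider second swing axis.
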PhysicsAsset /BRCosmetics/Characters/Player/Male/Medium/Bodies/M_MED_AshenMagnus_Slant/Meshes/Parts/M_MED_AshenMagnus_Slant_FaceAcc_Physics
bodies {
 Hair_base [{shapeType=Sphere, radius=0.505, shapeName=Hair_base_sphere}]
 dyn_hair_front_l [{shapeType=Sphere, radius=1, shapeName=dyn_hair_front_l_sphere}]
 dyn_hair_front_mid [{shapeType=Sphere, radius=1, shapeName=dyn_hair_front_mid_sphere}]
 dyn_hair_front_r [{shapeType=Sphere, radius=1, shapeName=dyn_hair_front_r_sphere}]
 dyn_hair_top_l_01 [{shapeType=Sphere, radius=1, shapeName=dyn_hair_top_l_01_sphere, center=<0.4436 0 0>}]
 dyn_hair_top_l_02 [{shapeType=Sphere, radius=1, shapeName=dyn_hair_top_l_02_sphere, center=<0.4436 0 0>}]
 dyn_hair_top_l_03 [{shapeType=Sphere, radius=1, shapeName=dyn_hair_top_l_03_sphere, center=<0.4436 0 0>}]
 dyn_hair_top_r_01 [{shapeType=Sphere, radius=1, shapeName=dyn_hair_top_r_01_sphere, center=<0.4436 0 0>}]
 dyn_hair_top_r_02 [{shapeType=Sphere, radius=1, shapeName=dyn_hair_top_r_02_sphere, center=<0.4436 0 0>}]
 dyn_hair_top_r_03 [{shapeType=Sphere, radius=1, shapeName=dyn_hair_top_r_03_sphere, center=<0.4436 0 0>}]
}
constraints [
 {bone1Name=dyn_hair_front_l, bone2Name=Hair_base, swing1LimitDegrees=10, swing2LimitDegrees=10, twistLimitDegrees=1}
 {bone1Name=dyn_hair_front_mid, bone2Name=Hair_base, swing1LimitDegrees=10, swing2LimitDegrees=10, twistLimitDegrees=1}
 {bone1Name=dyn_hair_front_r, bone2Name=Hair_base, swing1LimitDegrees=10, swing2LimitDegrees=10, twistLimitDegrees=1}
 {bone1Name=dyn_hair_top_l_01, bone2Name=Hair_base, swing1LimitDegrees=10, swing2LimitDegrees=20, twistLimitDegrees=1}
 {bone1Name=dyn_hair_top_l_02, bone2Name=Hair_base, swing1LimitDegrees=10, swing2LimitDegrees=20, twistLimitDegrees=1}
 {bone1Name=dyn_hair_top_l_03, bone2Name=Hair_base, swing1LimitDegrees=10, swing2LimitDegrees=20, twistLimitDegrees=1}
 {bone1Name=dyn_hair_top_r_01, bone2Name=Hair_base, swing1LimitDegrees=10, swing2LimitDegrees=20, twistLimitDegrees=1}
 {bone1Name=dyn_hair_top_r_02, bone2Name=Hair_base, swing1LimitDegrees=10, swing2LimitDegrees=20, twistLimitDegrees=1}
 {bone1Name=dyn_hair_top_r_03, bone2Name=Hair_base, swing1LimitDegrees=10, swing2LimitDegrees=20, twistLimitDegrees=1}
]
